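# Physics asset for the ElegantLily Plush mic stand cosmetic, defining physical motion for its flower and leaf bunches.
# Capsule bodies cover the long handle, eight bunch roots, and their stems, leaves and flowers.
# Bunches 02–08 lock rigidly to the handle, while bunch_01 has no handle constraint.
# Most child joints allow only small limited swings with locked twist, flowers generally stiffer than leaves, giving subtle sway.
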PhysicsAsset /SparksCosmetics/Instrument/MicStand/ElegantLily_Plush/Meshes/ElegantLily_Plush_MicStand_Physics
bodies {
 bunch_01 [{shapeType=Capsule, length=0, radius=1, shapeName=bunch_01_capsule}]
 bunch_01_leaf_A [{shapeType=Capsule, length=4.5, radius=2.75, shapeName=bunch_01_leaf_A_capsule, center=<4.7 0 0.7>, rotation=<rot 0 -90 -82>}]
 bunch_01_leaf_B [{shapeType=Capsule, length=4.5, radius=2, shapeName=bunch_01_leaf_B_capsule, center=<3.8 0 0.6>, rotation=<rot -90 0 0>}]
 bunch_01_leaf_C [{shapeType=Capsule, length=4.8, radius=2, shapeName=bunch_01_leaf_C_capsule, center=<3.75 0 0>, rotation=<rot 90 0 0>}]
 bunch_01_stem_leaf_A [{shapeType=Capsule, length=2, radius=1, shapeName=bunch_01_stem_leaf_A_capsule, center=<1 0 0>, rotation=<rot 90 0 0>}]
 bunch_02 [{shapeType=Capsule, length=0, radius=1}]
 bunch_02_flower [{shapeType=Capsule, length=4, radius=4, shapeName=bunch_02_flower_capsule, center=<4 0 0>}]
 bunch_02_leaf_A [{shapeType=Capsule, length=3.4, radius=2.5, shapeName=bunch_02_leaf_A_capsule, center=<4.7 0 0.25>, rotation=<rot -80 33 -38>}]
 bunch_02_leaf_B [{shapeType=Capsule, length=3.45, radius=2, shapeName=bunch_02_leaf_B_capsule, center=<3.6 0 -0.6>, rotation=<rot -90 0 0>}]
 bunch_02_stem_flower [{shapeType=Capsule, length=2.2, radius=0.7, shapeName=bunch_02_stem_flower_capsule, center=<1.7 0 0.3>, rotation=<rot 90 0 0>}]
 bunch_03 [{shapeType=Capsule, length=0, radius=1, shapeName=bunch_03_capsule}]
 bunch_03_leaf_A [{shapeType=Capsule, length=3.2918, radius=1.237, shapeName=bunch_03_leaf_A_capsule, center=<2.7 0 1>, rotation=<rot -90 180 -180>}]
 bunch_03_leaf_B [{shapeType=Capsule, length=2.5, radius=2, shapeName=bunch_03_leaf_B_capsule, center=<3.7 0 1>, rotation=<rot -88 0 0>}]
 bunch_03_leaf_C [{shapeType=Capsule, length=2.5, radius=2.7, shapeName=bunch_03_leaf_C_capsule, center=<4.5 0 1.75>, rotation=<rot -88 0 0>}]
 bunch_04 [{shapeType=Capsule, length=0, radius=1}]
 bunch_04_flower [{shapeType=Capsule, length=2.7398, radius=4.687, shapeName=bunch_04_flower_capsule, center=<1.8007 -0.0405 1.0323>, rotation=<rot 0 1 -3.4>}]
 bunch_04_leaf_A [{shapeType=Capsule, length=3.5, radius=2, shapeName=bunch_04_leaf_A_capsule, center=<4 0 0>, rotation=<rot -90 0 0>}]
 bunch_04_leaf_B [{shapeType=Capsule, length=1.8, radius=3.5, shapeName=bunch_04_leaf_B_capsule, center=<3.8 0 0>, rotation=<rot -90 -180 180>}]
 bunch_04_leaf_C [{shapeType=Capsule, length=4, radius=3, shapeName=bunch_04_leaf_C_capsule, center=<4.7 0 0.55>, rotation=<rot -90 -180 180>}]
 bunch_04_stem_leaf_B [{shapeType=Capsule, length=2, radius=1, shapeName=bunch_04_stem_leaf_B_capsule, center=<1 0 0>, rotation=<rot -90 -180 180>}]
 bunch_05 [{shapeType=Capsule, length=0, radius=1}]
 bunch_05_flower [{shapeType=Capsule, length=3.6, radius=6, shapeName=bunch_05_flower_capsule, center=<4.36 0 0>}]
 bunch_05_leaf_A [{shapeType=Capsule, length=5, radius=3, shapeName=bunch_05_leaf_A_capsule, center=<4.4 0 0>, rotation=<rot -90 180 -180>}]
 bunch_05_leaf_B [{shapeType=Capsule, length=5, radius=2.8, shapeName=bunch_05_leaf_B_capsule, center=<4.7 0 0>, rotation=<rot -90 0 0>}]
 bunch_05_leaf_C [{shapeType=Capsule, length=3.5, radius=3.5, shapeName=bunch_05_leaf_C_capsule, center=<4 0 0>, rotation=<rot -90 -180 180>}]
 bunch_05_leaf_D [{shapeType=Capsule, length=5.2622, radius=2.0218, shapeName=bunch_05_leaf_D_capsule, center=<4.4 0 0>, rotation=<rot -90 0 0>}]
 bunch_05_leaf_E [{shapeType=Capsule, length=3.8, radius=2, shapeName=bunch_05_leaf_E_capsule, center=<3.1 0 0>, rotation=<rot 70 -180 -180>}]
 bunch_05_stem_leaf_C [{shapeType=Capsule, length=1.3, radius=0.7, shapeName=bunch_05_stem_leaf_C_capsule, rotation=<rot -90 0 0>}]
 bunch_06 [{shapeType=Capsule, length=0, radius=1}]
 bunch_06_leaf_A [{shapeType=Capsule, length=4.5, radius=2, shapeName=bunch_06_leaf_A_capsule, center=<2.6 0 0.75>, rotation=<rot -90 0 0>}]
 bunch_06_leaf_B [{shapeType=Capsule, length=7, radius=2, shapeName=bunch_06_leaf_B_capsule, center=<4.75 0 0.6>, rotation=<rot -90 180 180>}]
 bunch_06_leaf_C [{shapeType=Capsule, length=4.3781, radius=2.3219, shapeName=bunch_06_leaf_C_capsule, center=<4 0 0>, rotation=<rot 90 0 0>}]
 bunch_06_leaf_D [{shapeType=Capsule, length=2.5, radius=3.5, shapeName=bunch_06_leaf_D_capsule, center=<4 0 0>, rotation=<rot 90 0 0>}]
 bunch_06_stem_leaf_D [{shapeType=Capsule, length=3, radius=1, shapeName=bunch_06_stem_leaf_D_capsule, center=<1.2 0 0>, rotation=<rot 90 0 0>}]
 bunch_07 [{shapeType=Capsule, length=0, radius=1}]
 bunch_07_flower [{shapeType=Capsule, length=4, radius=4, shapeName=bunch_07_flower_capsule}]
 bunch_07_leaf_A [{shapeType=Capsule, length=3.5, radius=3, shapeName=bunch_07_leaf_A_capsule, center=<5.5 0 0>, rotation=<rot 90 180 180>}]
 bunch_07_leaf_B [{shapeType=Capsule, length=4.4, radius=3.2, shapeName=bunch_07_leaf_B_capsule, center=<5.4 0 0>, rotation=<rot 90 0 0>}]
 bunch_08 [{shapeType=Capsule, length=0, radius=1}]
 bunch_08_A [{shapeType=Capsule, length=3.8, radius=3.8, shapeName=bunch_08_A_capsule, center=<5.5 0 0>, rotation=<rot 90 0 0>}]
 bunch_08_B [{shapeType=Capsule, length=3.8, radius=3, shapeName=bunch_08_B_capsule, center=<5 0 0.75>, rotation=<rot 90 0 0>}]
 bunch_08_flower [{shapeType=Capsule, length=5, radius=4.8, shapeName=bunch_08_flower_capsule, center=<4 0 0>}]
 handle [{shapeType=Capsule, length=93.4354, radius=1.6053, center=<0.2898 0.5047 21.0437>, rotation=<rot -0.8799 -90 -0.7627>}]
}
constraints [
 {
  bone1Name=bunch_02
  bone2Name=handle
  swing1LimitDegrees=45
  swing1Motion=Locked
  swing2LimitDegrees=45
  swing2Motion=Locked
  twistLimitDegrees=45
  twistMotion=Locked
 }
 {
  bone1Name=bunch_04
  bone2Name=handle
  swing1LimitDegrees=45
  swing1Motion=Locked
  swing2LimitDegrees=45
  swing2Motion=Locked
  twistLimitDegrees=45
  twistMotion=Locked
 }
 {
  bone1Name=bunch_05
  bone2Name=handle
  swing1LimitDegrees=45
  swing1Motion=Locked
  swing2LimitDegrees=45
  swing2Motion=Locked
  twistLimitDegrees=45
  twistMotion=Locked
 }
 {
  bone1Name=bunch_06
  bone2Name=handle
  swing1LimitDegrees=45
  swing1Motion=Locked
  swing2LimitDegrees=45
  swing2Motion=Locked
  twistLimitDegrees=45
  twistMotion=Locked
 }
 {
  bone1Name=bunch_07
  bone2Name=handle
  swing1LimitDegrees=45
  swing1Motion=Locked
  swing2LimitDegrees=45
  swing2Motion=Locked
  twistLimitDegrees=45
  twistMotion=Locked
 }
 {
  bone1Name=bunch_08
  bone2Name=handle
  swing1LimitDegrees=45
  swing1Motion=Locked
  swing2LimitDegrees=45
  swing2Motion=Locked
  twistLimitDegrees=45
  twistMotion=Locked
 }
 {bone1Name=bunch_01_stem_leaf_A, bone2Name=bunch_01, swing1LimitDegrees=0, swing2LimitDegrees=6, twistLimitDegrees=45, twistMotion=Locked}
 {
  bone1Name=bunch_01_leaf_A
  bone2Name=bunch_01_stem_leaf_A
  swing1LimitDegrees=10
  swing2LimitDegrees=8
  twistLimitDegrees=45
  twistMotion=Locked
 }
 {bone1Name=bunch_01_leaf_B, bone2Name=bunch_01, swing1LimitDegrees=8, swing2LimitDegrees=10, twistLimitDegrees=45, twistMotion=Locked}
 {bone1Name=bunch_01_leaf_C, bone2Name=bunch_01, swing1LimitDegrees=6, swing2LimitDegrees=8, twistLimitDegrees=45, twistMotion=Locked}
 {bone1Name=bunch_02_stem_flower, bone2Name=bunch_02, swing1LimitDegrees=1, swing2LimitDegrees=1, twistLimitDegrees=45, twistMotion=Locked}
 {bone1Name=bunch_02_leaf_A, bone2Name=bunch_02, swing1LimitDegrees=6, swing2LimitDegrees=6, twistLimitDegrees=45, twistMotion=Locked}
 {bone1Name=bunch_02_leaf_B, bone2Name=bunch_02, swing1LimitDegrees=6, swing2LimitDegrees=6, twistLimitDegrees=45, twistMotion=Locked}
 {
  bone1Name=bunch_02_flower
  bone2Name=bunch_02_stem_flower
  swing1LimitDegrees=1
  swing2LimitDegrees=1
  twistLimitDegrees=45
  twistMotion=Locked
 }
 {
  bone1Name=bunch_03
  bone2Name=handle
  swing1LimitDegrees=45
  swing1Motion=Locked
  swing2LimitDegrees=45
  swing2Motion=Locked
  twistLimitDegrees=45
  twistMotion=Locked
 }
 {bone1Name=bunch_03_leaf_A, bone2Name=bunch_03, swing1LimitDegrees=4, swing2LimitDegrees=4, twistLimitDegrees=45, twistMotion=Locked}
 {bone1Name=bunch_03_leaf_B, bone2Name=bunch_03, swing1LimitDegrees=8, swing2LimitDegrees=6, twistLimitDegrees=45, twistMotion=Locked}
 {bone1Name=bunch_03_leaf_C, bone2Name=bunch_03, swing1LimitDegrees=6, swing2LimitDegrees=8, twistLimitDegrees=45, twistMotion=Locked}
 {
  bone1Name=bunch_04_stem_leaf_B
  bone2Name=bunch_04
  swing1LimitDegrees=45
  swing1Motion=Locked
  swing2LimitDegrees=45
  swing2Motion=Locked
  twistLimitDegrees=45
  twistMotion=Locked
 }
 {bone1Name=bunch_04_leaf_A, bone2Name=bunch_04, swing1LimitDegrees=10, swing2LimitDegrees=10, twistLimitDegrees=45, twistMotion=Locked}
 {bone1Name=bunch_04_leaf_C, bone2Name=bunch_04, swing1LimitDegrees=6, swing2LimitDegrees=6, twistLimitDegrees=45, twistMotion=Locked}
 {bone1Name=bunch_04_flower, bone2Name=bunch_04, swing1LimitDegrees=2, swing2LimitDegrees=2, twistLimitDegrees=45, twistMotion=Locked}
 {
  bone1Name=bunch_04_leaf_B
  bone2Name=bunch_04_stem_leaf_B
  swing1LimitDegrees=8
  swing2LimitDegrees=4
  twistLimitDegrees=45
  twistMotion=Locked
 }
 {bone1Name=bunch_05_leaf_A, bone2Name=bunch_05, swing1LimitDegrees=10, swing2LimitDegrees=6, twistLimitDegrees=45, twistMotion=Locked}
 {bone1Name=bunch_05_leaf_B, bone2Name=bunch_05, swing1LimitDegrees=10, swing2LimitDegrees=6, twistLimitDegrees=45, twistMotion=Locked}
 {bone1Name=bunch_05_stem_leaf_C, bone2Name=bunch_05, swing1LimitDegrees=4, swing2LimitDegrees=4, twistLimitDegrees=45, twistMotion=Locked}
 {
  bone1Name=bunch_05_leaf_C
  bone2Name=bunch_05_stem_leaf_C
  swing1LimitDegrees=12
  swing2LimitDegrees=10
  twistLimitDegrees=45
  twistMotion=Locked
 }
 {bone1Name=bunch_05_leaf_D, bone2Name=bunch_05, swing1LimitDegrees=10, swing2LimitDegrees=6, twistLimitDegrees=45, twistMotion=Locked}
 {bone1Name=bunch_05_leaf_E, bone2Name=bunch_05, swing1LimitDegrees=8, swing2LimitDegrees=4, twistLimitDegrees=45, twistMotion=Locked}
 {bone1Name=bunch_05_flower, bone2Name=bunch_05, swing1LimitDegrees=4, swing2LimitDegrees=4, twistLimitDegrees=45, twistMotion=Locked}
 {bone1Name=bunch_06_leaf_A, bone2Name=bunch_06, swing1LimitDegrees=8, swing2LimitDegrees=4, twistLimitDegrees=45, twistMotion=Locked}
 {bone1Name=bunch_06_leaf_B, bone2Name=bunch_06, swing1LimitDegrees=8, swing2LimitDegrees=6, twistLimitDegrees=45, twistMotion=Locked}
 {bone1Name=bunch_06_leaf_C, bone2Name=bunch_06, swing1LimitDegrees=6, swing2LimitDegrees=6, twistLimitDegrees=45, twistMotion=Locked}
 {bone1Name=bunch_06_stem_leaf_D, bone2Name=bunch_06, swing1LimitDegrees=4, swing2LimitDegrees=4, twistLimitDegrees=45, twistMotion=Locked}
 {bone1Name=bunch_06_leaf_D, bone2Name=bunch_06_stem_leaf_D, swing1LimitDegrees=8, swing2LimitDegrees=6, twistLimitDegrees=5}
 {bone1Name=bunch_07_leaf_A, bone2Name=bunch_07, swing1LimitDegrees=10, swing2LimitDegrees=8, twistLimitDegrees=45, twistMotion=Locked}
 {bone1Name=bunch_07_leaf_B, bone2Name=bunch_07, swing1LimitDegrees=10, swing2LimitDegrees=8, twistLimitDegrees=45, twistMotion=Locked}
 {bone1Name=bunch_07_flower, bone2Name=bunch_07, swing1LimitDegrees=4, swing2LimitDegrees=4, twistLimitDegrees=45, twistMotion=Locked}
 {bone1Name=bunch_08_A, bone2Name=bunch_08, swing1LimitDegrees=10, swing2LimitDegrees=10, twistLimitDegrees=45, twistMotion=Locked}
 {bone1Name=bunch_08_B, bone2Name=bunch_08, swing1LimitDegrees=8, swing2LimitDegrees=8, twistLimitDegrees=45, twistMotion=Locked}
 {bone1Name=bunch_08_flower, bone2Name=bunch_08, swing1LimitDegrees=2, swing2LimitDegrees=2, twistLimitDegrees=45, twistMotion=Locked}
]
