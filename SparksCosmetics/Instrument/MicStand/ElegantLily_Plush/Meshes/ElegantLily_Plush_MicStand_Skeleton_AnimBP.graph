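# Animation Blueprint for the ElegantLily Plush mic stand cosmetic, adding physics and procedural sway to its leaf and flower bones.
# Each update, the EventGraph sets seven wave variables from real-time sine waves of differing frequencies, with amplitudes re-randomized every time.
# The AnimGraph runs the cached input pose through a RigidBody node, then chained additive Modify Bone nodes apply those waves as rotations.
# A second RigidBody node with stronger motion scales sits unconnected.
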AnimBlueprint /SparksCosmetics/Instrument/MicStand/ElegantLily_Plush/Meshes/ElegantLily_Plush_MicStand_Skeleton_AnimBP
target_skeleton=&/ElegantLily_Plush_MicStand_Skeleton
graphs [
 AnimGraph {nodes [
   RigidBody() {
    bOverrideWorldGravity=true
    overridePhysicsAsset=&/ElegantLily_Plush_MicStand_Physics
    simulationSpace=BaseBoneSpace
    alphaBoolBlend {blendOption=Linear}
    baseBoneRef {boneName=root}
    componentAppliedLinearAccClamp=<3000 3000 3000>
    componentLinearAccScale=<0.4 0.3 0.3>
    componentLinearVelScale=<0.4 0.3 0.3>
    overrideWorldGravity {z=-980}
    simSpaceSettings {worldAlpha=1, externalLinearVelocity {y=500, z=500}}
   }
   n1 = RigidBody(ComponentPose=Local To Component(LocalPose=Use cached pose 'cached input pose'())) {
    bOverrideWorldGravity=true
    overridePhysicsAsset=&ElegantLily_Plush_MicStand_Physics
    simulationSpace=BaseBoneSpace
    alphaBoolBlend {blendOption=Linear}
    baseBoneRef {boneName=root}
    componentAppliedLinearAccClamp=<1000 1000 1000>
    componentLinearAccScale=<0.2 0.2 0.2>
    componentLinearVelScale=<0.1 0.1 0.1>
    overrideWorldGravity {z=-980}
    simSpaceSettings {worldAlpha=1, externalLinearVelocity {y=500, z=500}}
   }
   n2 = "Transform (Modify) Bone - Bone: bunch_01_leaf_A"!(
    ComponentPose=@n1
    bAlphaBoolEnabled:bool=True
    Translation="0,0,0"
    Rotation_Roll:float=0
    Rotation_Pitch:float=Get wave_03() -> (wave_03:double)
    Rotation_Yaw:float=Get wave_shake() -> (wave_shake:double)
    Scale="1,1,1"
   ) -> (Pose) {
    alphaInputType=Bool
    rotationMode=BMM_Additive
    rotationSpace=BCS_ParentBoneSpace
    alphaBoolBlend {blendOption=Linear}
    boneToModify {boneName=bunch_01_leaf_A}
   }
   n3 = "Transform (Modify) Bone - Bone: bunch_01_leaf_B"!(
    ComponentPose=@n2
    bAlphaBoolEnabled:bool=True
    Translation="0,0,0"
    Rotation_Roll:float=0
    Rotation_Pitch:float=Get wave_01() -> (wave_01:double)
    Rotation_Yaw:float=Get wave_shake()
    Scale="1,1,1"
   ) -> (Pose) {
    alphaInputType=Bool
    rotationMode=BMM_Additive
    rotationSpace=BCS_ParentBoneSpace
    alphaBoolBlend {blendOption=Linear}
    boneToModify {boneName=bunch_01_leaf_B}
   }
   n4 = "Transform (Modify) Bone - Bone: bunch_01_leaf_C"!(
    ComponentPose=@n3
    bAlphaBoolEnabled:bool=True
    Translation="0,0,0"
    Rotation_Roll:float=0
    Rotation_Pitch:float=Get Wave 04() -> (Wave 04:double)
    Rotation_Yaw:float=Get wave_shake()
    Scale="1,1,1"
   ) -> (Pose) {
    alphaInputType=Bool
    rotationMode=BMM_Additive
    rotationSpace=BCS_ParentBoneSpace
    alphaBoolBlend {blendOption=Linear}
    boneToModify {boneName=bunch_01_leaf_C}
   }
   n5 = "Transform (Modify) Bone - Bone: bunch_02_flower"!(
    ComponentPose=@n4
    bAlphaBoolEnabled:bool=True
    Translation="0,0,0"
    Rotation_Roll:float=0
    Rotation_Pitch:float=Get wave_01()
    Rotation_Yaw:float=0
    Scale="1,1,1"
   ) -> (Pose) {
    alphaInputType=Bool
    rotationMode=BMM_Additive
    rotationSpace=BCS_ParentBoneSpace
    alphaBoolBlend {blendOption=Linear}
    boneToModify {boneName=bunch_02_flower}
   }
   n6 = "Transform (Modify) Bone - Bone: bunch_02_leaf_A"!(
    ComponentPose=@n5
    bAlphaBoolEnabled:bool=True
    Translation="0,0,0"
    Rotation_Roll:float=0
    Rotation_Pitch:float=Get wave_02() -> (wave_02:double)
    Rotation_Yaw:float=Get wave_shake()
    Scale="1,1,1"
   ) -> (Pose) {
    alphaInputType=Bool
    rotationMode=BMM_Additive
    rotationSpace=BCS_ParentBoneSpace
    alphaBoolBlend {blendOption=Linear}
    boneToModify {boneName=bunch_02_leaf_A}
   }
   n7 = "Transform (Modify) Bone - Bone: bunch_02_leaf_B"!(
    ComponentPose=@n6
    bAlphaBoolEnabled:bool=True
    Translation="0,0,0"
    Rotation_Roll:float=0
    Rotation_Pitch:float=Get Wave 05() -> (Wave 05:double)
    Rotation_Yaw:float=0
    Scale="1,1,1"
   ) -> (Pose) {
    alphaInputType=Bool
    rotationMode=BMM_Additive
    rotationSpace=BCS_ParentBoneSpace
    alphaBoolBlend {blendOption=Linear}
    boneToModify {boneName=bunch_02_leaf_B}
   }
   n8 = "Transform (Modify) Bone - Bone: bunch_03_leaf_A"!(
    ComponentPose=@n7
    bAlphaBoolEnabled:bool=True
    Translation="0,0,0"
    Rotation_Roll:float=0
    Rotation_Pitch:float=Get wave_01()
    Rotation_Yaw:float=Get wave_shake()
    Scale="1,1,1"
   ) -> (Pose) {
    alphaInputType=Bool
    rotationMode=BMM_Additive
    rotationSpace=BCS_ParentBoneSpace
    alphaBoolBlend {blendOption=Linear}
    boneToModify {boneName=bunch_03_leaf_A}
   }
   n9 = "Transform (Modify) Bone - Bone: bunch_03_leaf_B"!(
    ComponentPose=@n8
    bAlphaBoolEnabled:bool=True
    Translation="0,0,0"
    Rotation_Roll:float=0
    Rotation_Pitch:float=Get Wave 06() -> (Wave 06:double)
    Rotation_Yaw:float=Get wave_shake()
    Scale="1,1,1"
   ) -> (Pose) {
    alphaInputType=Bool
    rotationMode=BMM_Additive
    rotationSpace=BCS_ParentBoneSpace
    alphaBoolBlend {blendOption=Linear}
    boneToModify {boneName=bunch_03_leaf_B}
   }
   n10 = "Transform (Modify) Bone - Bone: bunch_03_leaf_C"!(
    ComponentPose=@n9
    bAlphaBoolEnabled:bool=True
    Translation="0,0,0"
    Rotation_Roll:float=0
    Rotation_Pitch:float=Get Wave 04()
    Rotation_Yaw:float=Get wave_shake()
    Scale="1,1,1"
   ) -> (Pose) {
    alphaInputType=Bool
    rotationMode=BMM_Additive
    rotationSpace=BCS_ParentBoneSpace
    alphaBoolBlend {blendOption=Linear}
    boneToModify {boneName=bunch_03_leaf_C}
   }
   n11 = "Transform (Modify) Bone - Bone: bunch_04_leaf_A"!(
    ComponentPose=@n10
    bAlphaBoolEnabled:bool=True
    Translation="0,0,0"
    Rotation_Roll:float=0
    Rotation_Pitch:float=Get wave_03()
    Rotation_Yaw:float=Get wave_shake()
    Scale="1,1,1"
   ) -> (Pose) {
    alphaInputType=Bool
    rotationMode=BMM_Additive
    rotationSpace=BCS_ParentBoneSpace
    alphaBoolBlend {blendOption=Linear}
    boneToModify {boneName=bunch_04_leaf_A}
   }
   n12 = "Transform (Modify) Bone - Bone: bunch_04_leaf_B"!(
    ComponentPose=@n11
    bAlphaBoolEnabled:bool=True
    Translation="0,0,0"
    Rotation_Roll:float=0
    Rotation_Pitch:float=Get Wave 05()
    Rotation_Yaw:float=Get wave_shake()
    Scale="1,1,1"
   ) -> (Pose) {
    alphaInputType=Bool
    rotationMode=BMM_Additive
    rotationSpace=BCS_ParentBoneSpace
    alphaBoolBlend {blendOption=Linear}
    boneToModify {boneName=bunch_04_leaf_B}
   }
   n13 = "Transform (Modify) Bone - Bone: bunch_04_leaf_C"!(
    ComponentPose=@n12
    bAlphaBoolEnabled:bool=True
    Translation="0,0,0"
    Rotation_Roll:float=0
    Rotation_Pitch:float=Get wave_02()
    Rotation_Yaw:float=Get wave_shake()
    Scale="1,1,1"
   ) -> (Pose) {
    alphaInputType=Bool
    rotationMode=BMM_Additive
    rotationSpace=BCS_ParentBoneSpace
    alphaBoolBlend {blendOption=Linear}
    boneToModify {boneName=bunch_04_leaf_C}
   }
   n14 = "Transform (Modify) Bone - Bone: bunch_04_flower"!(
    ComponentPose=@n13
    bAlphaBoolEnabled:bool=True
    Translation="0,0,0"
    Rotation_Roll:float=0
    Rotation_Pitch:float=0
    Rotation_Yaw:float=Get wave_01()
    Scale="1,1,1"
   ) -> (Pose) {
    alphaInputType=Bool
    rotationMode=BMM_Additive
    rotationSpace=BCS_ParentBoneSpace
    alphaBoolBlend {blendOption=Linear}
    boneToModify {boneName=bunch_04_flower}
   }
   n15 = "Transform (Modify) Bone - Bone: bunch_05_leaf_A"!(
    ComponentPose=@n14
    bAlphaBoolEnabled:bool=True
    Translation="0,0,0"
    Rotation_Roll:float=0
    Rotation_Pitch:float=Get wave_01()
    Rotation_Yaw:float=0
    Scale="1,1,1"
   ) -> (Pose) {
    alphaInputType=Bool
    rotationMode=BMM_Additive
    rotationSpace=BCS_ParentBoneSpace
    alphaBoolBlend {blendOption=Linear}
    boneToModify {boneName=bunch_05_leaf_A}
   }
   n16 = "Transform (Modify) Bone - Bone: bunch_05_leaf_B"!(
    ComponentPose=@n15
    bAlphaBoolEnabled:bool=True
    Translation="0,0,0"
    Rotation_Roll:float=0
    Rotation_Pitch:float=Get wave_03()
    Rotation_Yaw:float=Get wave_shake()
    Scale="1,1,1"
   ) -> (Pose) {
    alphaInputType=Bool
    rotationMode=BMM_Additive
    rotationSpace=BCS_ParentBoneSpace
    alphaBoolBlend {blendOption=Linear}
    boneToModify {boneName=bunch_05_leaf_B}
   }
   n17 = "Transform (Modify) Bone - Bone: bunch_05_leaf_C"!(
    ComponentPose=@n16
    bAlphaBoolEnabled:bool=True
    Translation="0,0,0"
    Rotation_Roll:float=0
    Rotation_Pitch:float=Get Wave 05()
    Rotation_Yaw:float=0
    Scale="1,1,1"
   ) -> (Pose) {
    alphaInputType=Bool
    rotationMode=BMM_Additive
    rotationSpace=BCS_ParentBoneSpace
    alphaBoolBlend {blendOption=Linear}
    boneToModify {boneName=bunch_05_leaf_C}
   }
   n18 = "Transform (Modify) Bone - Bone: bunch_05_leaf_D"!(
    ComponentPose=@n17
    bAlphaBoolEnabled:bool=True
    Translation="0,0,0"
    Rotation_Roll:float=0
    Rotation_Pitch:float=Get Wave 04()
    Rotation_Yaw:float=0
    Scale="1,1,1"
   ) -> (Pose) {
    alphaInputType=Bool
    rotationMode=BMM_Additive
    rotationSpace=BCS_ParentBoneSpace
    alphaBoolBlend {blendOption=Linear}
    boneToModify {boneName=bunch_05_leaf_D}
   }
   n19 = "Transform (Modify) Bone - Bone: bunch_05_leaf_E"!(
    ComponentPose=@n18
    bAlphaBoolEnabled:bool=True
    Translation="0,0,0"
    Rotation_Roll:float=0
    Rotation_Pitch:float=Get wave_02()
    Rotation_Yaw:float=Get wave_shake()
    Scale="1,1,1"
   ) -> (Pose) {
    alphaInputType=Bool
    rotationMode=BMM_Additive
    rotationSpace=BCS_ParentBoneSpace
    alphaBoolBlend {blendOption=Linear}
    boneToModify {boneName=bunch_05_leaf_E}
   }
   n20 = "Transform (Modify) Bone - Bone: bunch_05_flower"!(
    ComponentPose=@n19
    bAlphaBoolEnabled:bool=True
    Translation="0,0,0"
    Rotation_Roll:float=0
    Rotation_Pitch:float=Get wave_01()
    Rotation_Yaw:float=0
    Scale="1,1,1"
   ) -> (Pose) {
    alphaInputType=Bool
    rotationMode=BMM_Additive
    rotationSpace=BCS_ParentBoneSpace
    alphaBoolBlend {blendOption=Linear}
    boneToModify {boneName=bunch_05_flower}
   }
   n21 = "Transform (Modify) Bone - Bone: bunch_06_leaf_A"!(
    ComponentPose=@n20
    bAlphaBoolEnabled:bool=True
    Translation="0,0,0"
    Rotation_Roll:float=0
    Rotation_Pitch:float=Get Wave 04()
    Rotation_Yaw:float=0
    Scale="1,1,1"
   ) -> (Pose) {
    alphaInputType=Bool
    rotationMode=BMM_Additive
    rotationSpace=BCS_ParentBoneSpace
    alphaBoolBlend {blendOption=Linear}
    boneToModify {boneName=bunch_06_leaf_A}
   }
   n22 = "Transform (Modify) Bone - Bone: bunch_06_leaf_B"!(
    ComponentPose=@n21
    bAlphaBoolEnabled:bool=True
    Translation="0,0,0"
    Rotation_Roll:float=Get wave_02()
    Rotation_Pitch:float=0
    Rotation_Yaw:float=0
    Scale="1,1,1"
   ) -> (Pose) {
    alphaInputType=Bool
    rotationMode=BMM_Additive
    rotationSpace=BCS_ParentBoneSpace
    alphaBoolBlend {blendOption=Linear}
    boneToModify {boneName=bunch_06_leaf_B}
   }
   n23 = "Transform (Modify) Bone - Bone: bunch_06_leaf_C"!(
    ComponentPose=@n22
    bAlphaBoolEnabled:bool=True
    Translation="0,0,0"
    Rotation_Roll:float=0
    Rotation_Pitch:float=Get Wave 05()
    Rotation_Yaw:float=Get wave_shake()
    Scale="1,1,1"
   ) -> (Pose) {
    alphaInputType=Bool
    rotationMode=BMM_Additive
    rotationSpace=BCS_ParentBoneSpace
    alphaBoolBlend {blendOption=Linear}
    boneToModify {boneName=bunch_06_leaf_C}
   }
   n24 = "Transform (Modify) Bone - Bone: bunch_06_leaf_D"!(
    ComponentPose=@n23
    bAlphaBoolEnabled:bool=True
    Translation="0,0,0"
    Rotation_Roll:float=Get wave_shake()
    Rotation_Pitch:float=Get wave_02()
    Rotation_Yaw:float=0
    Scale="1,1,1"
   ) -> (Pose) {
    alphaInputType=Bool
    rotationMode=BMM_Additive
    rotationSpace=BCS_ParentBoneSpace
    alphaBoolBlend {blendOption=Linear}
    boneToModify {boneName=bunch_06_leaf_D}
   }
   n25 = "Transform (Modify) Bone - Bone: bunch_07_leaf_A"!(
    ComponentPose=@n24
    bAlphaBoolEnabled:bool=True
    Translation="0,0,0"
    Rotation_Roll:float=0
    Rotation_Pitch:float=Get wave_03()
    Rotation_Yaw:float=Get wave_shake()
    Scale="1,1,1"
   ) -> (Pose) {
    alphaInputType=Bool
    rotationMode=BMM_Additive
    rotationSpace=BCS_ParentBoneSpace
    alphaBoolBlend {blendOption=Linear}
    boneToModify {boneName=bunch_07_leaf_A}
   }
   n26 = "Transform (Modify) Bone - Bone: bunch_07_leaf_B"!(
    ComponentPose=@n25
    bAlphaBoolEnabled:bool=True
    Translation="0,0,0"
    Rotation_Roll:float=Get wave_shake()
    Rotation_Pitch:float=Get Wave 05()
    Rotation_Yaw:float=0
    Scale="1,1,1"
   ) -> (Pose) {
    alphaInputType=Bool
    rotationMode=BMM_Additive
    rotationSpace=BCS_ParentBoneSpace
    alphaBoolBlend {blendOption=Linear}
    boneToModify {boneName=bunch_07_leaf_B}
   }
   n27 = "Transform (Modify) Bone - Bone: bunch_07_flower"!(
    ComponentPose=@n26
    bAlphaBoolEnabled:bool=True
    Translation="0,0,0"
    Rotation_Roll:float=0
    Rotation_Pitch:float=0
    Rotation_Yaw:float=Get wave_02()
    Scale="1,1,1"
   ) -> (Pose) {
    alphaInputType=Bool
    rotationMode=BMM_Additive
    rotationSpace=BCS_ParentBoneSpace
    alphaBoolBlend {blendOption=Linear}
    boneToModify {boneName=bunch_07_flower}
   }
   n28 = "Transform (Modify) Bone - Bone: bunch_08_A"!(
    ComponentPose=@n27
    bAlphaBoolEnabled:bool=True
    Translation="0,0,0"
    Rotation_Roll:float=0
    Rotation_Pitch:float=Get wave_01()
    Rotation_Yaw:float=Get wave_shake()
    Scale="1,1,1"
   ) -> (Pose) {
    alphaInputType=Bool
    rotationMode=BMM_Additive
    rotationSpace=BCS_ParentBoneSpace
    alphaBoolBlend {blendOption=Linear}
    boneToModify {boneName=bunch_08_A}
   }
   n29 = "Transform (Modify) Bone - Bone: bunch_08_B"!(
    ComponentPose=@n28
    bAlphaBoolEnabled:bool=True
    Translation="0,0,0"
    Rotation_Roll:float=0
    Rotation_Pitch:float=0
    Rotation_Yaw:float=Get Wave 06()
    Scale="1,1,1"
   ) -> (Pose) {
    alphaInputType=Bool
    rotationMode=BMM_Additive
    rotationSpace=BCS_ParentBoneSpace
    alphaBoolBlend {blendOption=Linear}
    boneToModify {boneName=bunch_08_B}
   }
   n30 = "Transform (Modify) Bone - Bone: bunch_08_flower"!(
    ComponentPose=@n29
    bAlphaBoolEnabled:bool=True
    Translation="0,0,0"
    Rotation_Roll:float=0
    Rotation_Pitch:float=0
    Rotation_Yaw:float=Get Wave 04()
    Scale="1,1,1"
   ) -> (Pose) {
    alphaInputType=Bool
    rotationMode=BMM_Additive
    rotationSpace=BCS_ParentBoneSpace
    alphaBoolBlend {blendOption=Linear}
    boneToModify {boneName=bunch_08_flower}
   }
   Output Pose(Result=Component To Local(ComponentPose=@n30))
   Save cached pose 'cached input pose'(Pose=Input Pose()) {cacheName=cached input pose}
   Get wave_shake()
   Get wave_shake()
   Get wave_shake()
   Get wave_shake()
   Get wave_shake()
   Get wave_shake()
   Get wave_shake()
   Get wave_shake()
   Get wave_shake()
   Get wave_shake()
   Get wave_shake()
   Get wave_shake()
  ]}
 EventGraph {nodes [
   Event BlueprintUpdateAnimation()
   > Sequence()
   n1 = GetRealTimeSeconds()
   n2 = float * float(A=@n1, B=4)
   > Set wave_01(wave_01:double=float * float(A="Sin (Radians)"(A=@n2), B=RandomFloatInRange(Min=0.5, Max=0.75))) -> (Output_Get:double)
   n3 = float * float(A=@n1, B=6)
   > Set wave_02(wave_02:double=float * float(A="Sin (Radians)"(A=@n3), B=RandomFloatInRange(Min=1, Max=1.5))) -> (Output_Get:double)
   n4 = float * float(A=@n1, B=8)
   > Set wave_03(wave_03:double=float * float(A="Sin (Radians)"(A=@n4), B=RandomFloatInRange(Min=1.5, Max=1.75))) -> (Output_Get:double)
   n5 = float * float(A=@n1, B=1)
   > Set wave_shake(wave_shake:double=float * float(A="Sin (Radians)"(A=@n5), B=RandomFloatInRange(Min=0.1, Max=0.5))) -> (Output_Get:double)
   n6 = float * float(A=@n1, B=4.2)
   > Set Wave 04(Wave 04:double=float * float(A="Sin (Radians)"(A=@n6), B=RandomFloatInRange(Min=-0.5, Max=-0.75))) -> (Output_Get:double)
   n7 = float * float(A=@n1, B=8.2)
   > Set Wave 05(Wave 05:double=float * float(A="Sin (Radians)"(A=@n7), B=RandomFloatInRange(Min=-1, Max=-1.5))) -> (Output_Get:double)
   n8 = float * float(A=@n1, B=10.2)
   > Set Wave 06(Wave 06:double=float * float(A="Sin (Radians)"(A=@n8), B=RandomFloatInRange(Min=-1.5, Max=-1.75))) -> (Output_Get:double)
   TryGetPawnOwner()
  ]}
]
variables [
 Wave 04 {type=number}
 Wave 05 {type=number}
 Wave 06 {type=number}
 wave_01 {type=number}
 wave_02 {type=number}
 wave_03 {type=number}
 wave_shake {type=number}
]
